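; Parent Blueprint for the Dingo loot box, a projectile actor that drops loot when it stops simulating or enters water.
; Construction maps Rarity to a loot table and box color.
; On stop, the server rolls loot once and spawns pickups at an offset, plus toss ammo for weapons.
; Destruction is server-authoritative: a replicated flag hides the box on clients, and delayed actor destruction leaves time for gameplay cue FX.
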(event Custom|HideAndKill_Client
  ;[0.4896 0.2825 0.1492] HideAndKill
  ; Play GC, Hide Actor, disable collision, and Kill - replicates from server to client to better handle mis-triggered events
  (Rendering|SetActorHiddenInGame true)
  (Audio|Components|Audio|FadeOut (Variables|Default|GetAudio) 0.25 0.0 "Linear" :!class /Script/Engine.AudioComponent)
  (Collision|SetActorEnableCollision false))

(event OnEnterWaterEvent(FortWaterInteraction)
  (params (WaterBody :type Fort%20Water%20Body%20Actor%20Object%20Reference) (WaterInteractionComponent :type Fort%20Water%20Interaction%20Component%20Object%20Reference) (bIsFirstBody :type Boolean))
  (bind (_waterplanelocation _waterplanenormal _watersurfaceposition _waterdepth _waterbodyidx) (FX|GetLastWaterSurfaceInfo (Variables|Default|GetFortWaterInteraction) :!class /Script/FortniteGame.FortWaterInteractionComponent))
  ;[1 1 1] On Enter Water Stop simulating and play Splash FX
  (Game|Components|ProjectileMovement|StopSimulating (Variables|Default|GetProjectileMovement) (Variables|Default|GetNullHit))
  (Ability|ExecuteGameplayCueWithParamsReplicated (Variables|Default|GetGC_Splash) (Ability|GameplayCue|MakeGameplayCueParameters :NormalizedMagnitude (Math|Conversions|ToFloat(Integer) _waterbodyidx) :RawMagnitude _waterdepth :Location _waterplanelocation :Normal "0,0,0" :GameplayEffectLevel 1 :AbilityLevel 1 :bReplicateLocationWhenUsingMinimalRepProxy false))
  ;[1 1 1] Handle Stop Sim hiding and spawn Items
  (CallFunction|SpawnItems)
  (CallFunction|HideandKill))

(event Custom|HideAndKill
  ;[0.4896 0.2825 0.1492] HideAndKill
  ; Play GC, Hide Actor, disable collision, and Kill - replicates from server to client to better handle mis-triggered events
  (Networking|SwitchHasAuthority
    (:Authority
      (Utilities|FlowControl|Sequence
        (:then_0
          (Variables|Default|SetbReplicateHideAndKill true) ; replicate valid kill event from Server to clients
          (Ability|ExecuteGameplayCueReplicated (Variables|Default|GetGC_Destroy))
          (Collision|SetActorEnableCollision false))
        (:then_1
          (Utilities|FlowControl|Delay 1.0) ; short delay allowing the Clients \nto play destruction FX\nObject is effectively removed with \nvisibility and collision disabled
          (Actor|DestroyActor))))) ; make hide and kill server authoritative\nFixes FORT-792098 where the client Box\nwould collide with the despawning Drone\nand get destroyed while the server would \ncontinue to function as expected
)

(event EventStopSim
  (params (Hit :type Hit%20Result%20Structure))
  ;[1 1 1] Handle Stop Sim hiding and spawn Items
  (|CallParentFunction Hit)
  (CallFunction|SpawnItems)
  (CallFunction|HideandKill))

(event Custom|Spawn%20Items
  (bind _returnvalue (Transformation|GetActorLocation))
  (bind _spawnpositionoffset (Variables|Default|GetSpawnPositionOffset))
  (bind _returnvalue_2 (+ _returnvalue _spawnpositionoffset))
  ;[1 1 1] Spawn Items from array
  (if (Networking|IsServer)
    (Utilities|FlowControl|DoOnce false
      (:Completed
        (Development|PrintString "Spawn Items")
        (bind (_outloottodrop _returnvalue_3) (Loot|PickLootDrops (Variables|Default|GetItemsToDrop) 1 -1 "(GameplayTags=)"))
        (if _returnvalue_3
          (for _array_element _outloottodrop
            (bind (_array_element_itemdefinition _array_element_itemguid _array_element_count) (Class|FortItemFunctionLibrary|BreakFortItemEntry _array_element))
            (Development|PrintString (Utilities|String|Append "Item To Spawn:" (Utilities|GetDisplayName _array_element_itemdefinition)))
            (bind _asfort_world_item_definition (Utilities|Casting|CastToFortWorldItemDefinition _array_element_itemdefinition))
            (bind _returnvalue_4 (Inventory|SpawnPickupInWorld _asfort_world_item_definition _array_element_count _returnvalue_2 (Transformation|GetActorForwardVector) 0 true true false -1 "Container" "Drone"))
            (bind _asfort_weapon_item_definition (Utilities|Casting|CastToFortWeaponItemDefinition (Class|FortItemFunctionLibrary|BreakFortItemEntry (Pickup|GetItemEntry _returnvalue_4 0))))
            (Fortnite|Weapon|CreateTossAmmoPickupforWeaponItemDefinitionatLocation _asfort_weapon_item_definition (Variables|Default|GetT_AmmoTossTag) _returnvalue_2)))))))

(fn FindRarityToItemsAndColor
  (bind _rarity (Variables|Default|GetRarity))
  (bind _raritytoitemsmap (Variables|Default|GetRarityToItemsMap))
  (bind (_value _returnvalue) (Utilities|Map|Find _raritytoitemsmap _rarity))
  (bind _rarity_2 (Variables|Default|GetRarity))
  (bind _raritytocolormap (Variables|Default|GetRarityToColorMap))
  (bind (_value_2 _returnvalue_2) (Utilities|Map|Find _raritytocolormap _rarity_2))
  (if _returnvalue
    (Variables|Default|SetItemsToDrop _value))
  (if _returnvalue_2
    (Variables|Default|SetColorOfBox _value_2)))

(fn OnRep_bReplicateHideAndKill
  (if (Variables|Default|GetbReplicateHideAndKill)
    (Networking|CanExecuteCosmeticEvents
      (:True
        (CallFunction|HideandKillClient) ; replicate valid kill event from server to client
      ))))

(fn ConstructionScript
  (CallFunction|FindRaritytoItemsandColor)
  (|CallParentFunction))
