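; Parent Blueprint for the Dingo supply drop drone, which roams until destroyed, then spawns a rarity-based box and reports quest progress.
; Server timers steer its projectile movement through random turns, trace-based height following and an optional leash, with a failsafe for getting stuck.
; Side collisions trigger a multicast timeline turn, while vehicle hits and EMP events destroy it.
; It also handles storm despawn, dissolve effects and client-side interpolation.
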(event Collision|EventHit
  (params (MyComp :type Primitive%20Component%20Object%20Reference) (Other :type Actor%20Object%20Reference) (OtherComp :type Primitive%20Component%20Object%20Reference) (bSelfMoved :type Boolean) (HitLocation :type Vector :default "0, 0, 0") (HitNormal :type Vector :default "0, 0, 0") (NormalImpulse :type Vector :default "0, 0, 0") (Hit :type Hit%20Result%20Structure%20%28by%20ref%29))
  (bind (_returnvalue_x _returnvalue_y _returnvalue_z) (Math|Vector|BreakVector (Transformation|GetActorLocation Other)))
  (bind (_returnvalue_x_2 _returnvalue_y_2 _returnvalue_z_2) (Math|Vector|BreakVector (Transformation|GetActorLocation)))
  (bind (_returnvalue_roll _returnvalue_pitch _returnvalue_yaw) (Math|Rotator|BreakRotator (Math|Rotator|CombineRotators (Transformation|GetWorldRotation (Variables|Default|GetSphere)) (Math|Rotator|MakeRotator 0.0 0.0 (Math|Random|RandomFloatinRange (Variables|HitRot|GetHitRotTurnMin) (Variables|HitRot|GetHitRotTurnMax))))))
  ;[1 1 1 0.5] MOVEMENT
  ;;[1 1 1 0.75] Hit Event for Projectile Collision
  (if (Networking|IsServer)
    (CallFunction|CheckHitbyVehicle Other)
    (if (not (Variables|HitRot|GetHitRotating))
      (bind _asfort_player_pawn (Utilities|Casting|CastToFortPlayerPawn Other)
        (:then
          (Utilities|FrontCheck _asfort_player_pawn
            (:true
              (if (< _returnvalue_z _returnvalue_z_2)
                _))))
        (:CastFailed
          _))
      (Development|PrintString "Hit" (Variables|Default|GetDebug) (Variables|Default|GetDebug))
      (Utilities|CheckforHitToporBottom HitLocation
        (:SideHit
          (Ability|ExecuteGameplayCueWithParamsReplicated (Variables|GCs|GetGC_HitWindFXTag) (Ability|GameplayCue|MakeGameplayCueParameters :NormalizedMagnitude 1e+02 :RawMagnitude 1e+02 :Location HitLocation :Normal HitNormal :Instigator (Game|GetInstigator) :EffectCauser self :GameplayEffectLevel 1 :AbilityLevel 1 :bReplicateLocationWhenUsingMinimalRepProxy false))
          (Variables|HitRot|SetHitRotating true)
          (CallFunction|HitTurnMulticast 0 (Transformation|GetWorldRotation (Variables|Default|GetSphere)) (Math|Rotator|MakeRotator 0.0 0.0 _returnvalue_yaw) (Math|Random|RandomFloatinRange (Variables|HitRot|GetHitTurnPlayRateMin) (Variables|HitRot|GetHitTurnPlayRateMax))))
        (:TopHit
          _)
        (:BottomHit
          _))
      ;;;[1 1 1] If We take a top or bottom hit then fly straight
      (Projectile|SetDesiredDirection 0 0 (Transformation|GetActorForwardVector) :!class /Script/FortniteGame.FortProjectileMovementInterface))))

(event Custom|SetupLeashCheck
  ;[1 1 1 0.5] MOVEMENT
  ;;[1 1 1] Check if we hit our leash, if we did then turn around the Dingo
  (bind _returnvalue (Utilities|Time|SetTimerbyEvent (delegate Custom|LeashCheck) (Variables|LeashCheck|GetLeashCheckTime) true false 0.0 (* (Variables|LeashCheck|GetLeashCheckTime) (Variables|Default|GetVarianceAmountForTimers))))
  (Variables|LeashCheck|SetTimer_Leash _returnvalue))

(event Custom|LeashCheck
  (bind (_returnvalue_x _returnvalue_y _returnvalue_z) (Math|Vector|BreakVector (Transformation|GetActorLocation)))
  (bind (_startingloc_x _startingloc_y) (Math|Vector|BreakVector (Variables|GetStartingLoc_X)))
  ;[1 1 1 0.5] MOVEMENT
  ;;[1 1 1] Check if we hit our leash, if we did then turn around the Dingo
  (if (Variables|HitRot|GetHitRotating)
    _
    (elif (> (Math|Vector|Distance(Vector) (Math|Vector|MakeVector _startingloc_x _startingloc_y _returnvalue_z) (Math|Vector|MakeVector _returnvalue_x _returnvalue_y _returnvalue_z)) (Variables|LeashCheck|GetLeashDistance))
      (Development|PrintString "Hit Leash" (Variables|Default|GetDebug) (Variables|Default|GetDebug))
      (Utilities|IsValid (Variables|Default|GetFortProjectileMovement)
        (:Is%20Valid
          (Projectile|SetDesiredDirection 0 0 (Math|Vector|vector-vector (Math|Vector|MakeVector _startingloc_x _startingloc_y _returnvalue_z) (Math|Vector|MakeVector _returnvalue_x _returnvalue_y _returnvalue_z)) :!class /Script/FortniteGame.FortProjectileMovementInterface)
          (Variables|LeashCheck|SetLeashTurning true))))))

(event Custom|ResumeMovement
  (bind _fortprojectilemovement (Variables|Default|GetFortProjectileMovement))
  (bind _sphere (Variables|Default|GetSphere))
  ;[1 1 1 0.5] MOVEMENT
  ;;[1 1 1 0.75] Hit Event for Projectile Collision
  ;;;[1 1 1] Rotate Around When Hit
  (Utilities|IsValid _fortprojectilemovement
    (:Is%20Valid
      (Components|Movement|SetUpdatedComponent _fortprojectilemovement _sphere)
      (bind _output_get (Variables|/Script/Engine.MovementComponent|SetVelocity (* (Transformation|GetForwardVector _sphere) (Components|Movement|GetMaxSpeed (Variables|Default|GetFortProjectileMovement))) (Variables|Default|GetFortProjectileMovement)))
      (Projectile|SetDesiredDirection 0 0 _output_get :!class /Script/FortniteGame.FortProjectileMovementInterface)
      (Utilities|FlowControl|Delay (Variables|HitRot|GetDelayHitRotReset))
      (Variables|HitRot|SetHitRotating false))))

(event Custom|SetupHeightCheck
  ;[1 1 1 0.5] MOVEMENT
  ;;[1 1 1] Move Up and Down based on Geo
  (bind _returnvalue (Utilities|Time|SetTimerbyEvent (delegate Custom|HeightCheck) (Variables|HeightCheck|GetHeightCheckTime) true false 0.0 1.0))
  (Variables|HeightCheck|SetTimer_HeightCheck _returnvalue)
  (Variables|MovementFailsafe|SetLastLocationCheck (Transformation|GetActorLocation)))

(event Custom|HitTurnMulticast
  (params (StartRot :type Rotator) (EndRot :type Rotator) (PlayRate :type Float%20%28double-precision%29))
  (bind _fortprojectilemovement (Variables|Default|GetFortProjectileMovement))
  (bind _sphere (Variables|Default|GetSphere))
  (bind _fortprojectilemovement_2 (Variables|Default|GetFortProjectileMovement))
  ;[1 1 1 0.5] MOVEMENT
  ;;[1 1 1 0.75] Hit Event for Projectile Collision
  ;;;[1 1 1] Rotate Around When Hit
  (Utilities|IsValid _fortprojectilemovement_2
    (:Is%20Valid
      (Variables|/Script/Engine.MovementComponent|SetVelocity "0, 0, 0" _fortprojectilemovement_2)
      (Game|Components|ProjectileMovement|StopSimulating _fortprojectilemovement_2 (Variables|HitRot|GetNullHit))
      (Components|Timeline|SetPlayRate (Variables|BGA_Athena_Dingo_Parent|GetHitTurn) PlayRate)
      (bind (_direction _turntime) (Timeline|HitTurn PlayFromStart)
        (:Update
          (Transformation|SetWorldRotation (Variables|Default|GetSphere) (Math|Rotator|Lerp(Rotator) StartRot EndRot _turntime true) false false))
        (:Finished
          (Utilities|IsValid _fortprojectilemovement
            (:Is%20Valid
              (Components|Movement|SetUpdatedComponent _fortprojectilemovement _sphere)
              (bind _output_get (Variables|/Script/Engine.MovementComponent|SetVelocity (* (Transformation|GetForwardVector _sphere) (Components|Movement|GetMaxSpeed (Variables|Default|GetFortProjectileMovement))) (Variables|Default|GetFortProjectileMovement)))
              (Projectile|SetDesiredDirection 0 0 _output_get :!class /Script/FortniteGame.FortProjectileMovementInterface)
              (Utilities|FlowControl|Delay (Variables|HitRot|GetDelayHitRotReset))
              (Variables|HitRot|SetHitRotating false))))))))

(event Custom|HeightCheck
  ;[1 1 1 0.5] MOVEMENT
  ;;[1 1 1] Move Up and Down based on Geo
  (Ability|ExecuteGameplayCueReplicated (Variables|GCs|GetGC_AmbientSound)) ; Play a sound here with a chance to not play in the GC. This is to help the drone be found easier.
  (CallFunction|MovingCheck) ; Check if we are moving so we do not get stuck
  (if (and (not (Variables|HitRot|GetHitRotating)) (not (Variables|LeashCheck|GetLeashTurning)))
    (bind (_new_velocity _adjust) (CallFunction|HeightCheckCalc))
    (if _adjust
      (if (not (Variables|RandomTurn|GetIsTurning))
        (Utilities|IsValid (Variables|Default|GetFortProjectileMovement)
          (:Is%20Valid
            (Projectile|SetDesiredDirection 0 0 _new_velocity :!class /Script/FortniteGame.FortProjectileMovementInterface)))))))

(event Custom|SetupHeightPicker
  ;[1 1 1 0.5] MOVEMENT
  ;;[1 1 1] Set New Height
  (bind _returnvalue (Utilities|Time|SetTimerbyEvent (delegate Custom|SelectHeight) (Variables|HeightCheck|GetHeightSelectTime) true false 0.0 (* (Variables|HeightCheck|GetHeightSelectTime) (Variables|Default|GetVarianceAmountForTimers))))
  (Variables|HeightCheck|SetTimer_SelectHeight _returnvalue)
  (CallFunction|SelectHeight))

(event Custom|SetupRandomTurns
  ;[1 1 1 0.5] MOVEMENT
  ;;[1 1 1] Random Turning
  (Utilities|Time|SetTimerbyEvent (delegate Custom|TurnTrigger) (Variables|RandomTurn|GetRandomTurnTime) true false 0.0 (* (Variables|RandomTurn|GetRandomTurnTime) (Variables|Default|GetVarianceAmountForTimers))))

(event Custom|SelectHeight
  ;[1 1 1 0.5] MOVEMENT
  ;;[1 1 1] Set New Height
  (Variables|HeightCheck|SetTargetHeight (Math|Random|RandomFloatinRange (Variables|HeightCheck|GetFlyLow) (Variables|HeightCheck|GetFlyHigh))))

(event Custom|TurnTrigger
  (bind _fortprojectilemovement (Variables|Default|GetFortProjectileMovement))
  ;[1 1 1 0.5] MOVEMENT
  ;;[1 1 1] Random Turning
  (Utilities|CheckforCircle
    (:true
      (CallFunction|OutOfCircleDespawn))
    (:false
      _)
    (:CastFailed
      _))
  (if (Math|Boolean|ANDBoolean (Math|Random|RandomBoolwithWeight (Variables|RandomTurn|GetRandomTurnChance)) (not (Variables|HitRot|GetHitRotating)) (not (Variables|LeashCheck|GetLeashTurning)))
    (Utilities|IsValid _fortprojectilemovement
      (:Is%20Valid
        (Development|PrintString "Turn" (Variables|Default|GetDebug) (Variables|Default|GetDebug))
        (Projectile|SetDesiredDirection 0 0 (Math|Vector|RotateVectorAroundAxis (Math|Vector|Normalize (Variables|/Script/Engine.MovementComponent|GetVelocity _fortprojectilemovement)) (* (Math|Float|SelectFloat 1.0 -1.0 (Math|Random|RandomBool)) (Math|Random|RandomFloatinRange 3e+01 1.2e+02)) "0, 0,1.000000") :!class /Script/FortniteGame.FortProjectileMovementInterface)))))

(event Custom|OutOfCircleDespawn
  ;[1 1 1] Despawn if in storm
  (Utilities|Time|SetTimerbyEvent (delegate Custom|Despawn) (Variables|Despawning|GetTimeBeforeStormDespawn)))

(event Custom|Despawn
  ;[1 1 1] Despawn
  (Utilities|FlowControl|DoOnce false
    (:Completed
      (Variables|Despawning|SetDespawned true)
      (Collision|SetActorEnableCollision false)
      (Rendering|SetActorHiddenInGame true)
      (GameplayCue|ExecuteGameplayCue%28Burst%29Local self (Variables|GCs|GetGC_Despawn) (Ability|GameplayCue|MakeGameplayCueParameters :NormalizedMagnitude 0.0 :RawMagnitude 0.0 :Location (Transformation|GetActorLocation) :Normal (Transformation|GetActorForwardVector) :Instigator self :EffectCauser self :SourceObject self :GameplayEffectLevel 1 :AbilityLevel 1 :bReplicateLocationWhenUsingMinimalRepProxy false))
      (Actor|SetLifeSpan (Variables|Despawning|GetDespawnLifespan)))))

(event Custom|SpawnBox
  (bind (_returnvalue_roll _returnvalue_pitch _returnvalue_yaw) (Math|Rotator|BreakRotator (Transformation|GetActorRotation)))
  ;[1 1 1] Spawn Box
  (if (Networking|IsServer)
    (Game|SpawnActorfromClass (Variables|Default|GetBoxToSpawn) (Math|Transform|MakeTransform (Transformation|GetActorLocation) (Math|Rotator|MakeRotator 0.0 0.0 _returnvalue_yaw) "1.000000,1.000000,1.000000") "AlwaysSpawn" "SelectDefaultAtRuntime" (Actor|GetOwner) (Variables|Rarity|GetRarity) 0 0)))

(event Custom|HandleEMP
  (params (InstigatorController :type Controller%20Object%20Reference))
  ;[1 1 1] Handle EMP
  (EventDispatchers|UnbindEventfromOnDied (EventDispatchers|CreateEvent :!event Custom|Destroyed))
  (CallFunction|Destroyed 0 0.0 InstigatorController))

(event Custom|Destroyed
  (params (DamagedActor :type Actor%20Object%20Reference) (Damage :type Float%20%28single-precision%29) (InstigatedBy :type Controller%20Object%20Reference) (DamageCauser :type Actor%20Object%20Reference) (HitLocation :type Vector) (FHitComponent :type Primitive%20Component%20Object%20Reference) (BoneName :type Name) (Momentum :type Vector))
  ;[1 1 1] On Destroyed
  (if (not (Variables|Despawning|GetDespawned))
    (Variables|States|SetBeingDestroyed true)
    (Development|PrintString "Event Destroyed" (Variables|Default|GetDebug) (Variables|Default|GetDebug))
    ;;[1 1 1] Hide, Deactivate sounds, play death FX and spawn Items, Destroy Pickup
    (Rendering|SetActorHiddenInGame true)
    (Collision|SetActorEnableCollision false)
    (Ability|ExecuteGameplayCueReplicated (Variables|GCs|GetGC_Death))
    (CallFunction|SpawnBox)
    (Audio|Components|Audio|FadeOut (Variables|Audio|GetHoverSound) 0.15 0.0 "Linear" :!class /Script/Engine.AudioComponent)
    (CallFunction|QuestProgress InstigatedBy)
    (SoundIndicator|StopSoundIndicatorLooping (Variables|Default|GetFortSoundIndicator))))

(event Custom|OnSpawn
  ;[1 1 1] On Spawn
  (Variables|LeashCheck|SetStartingLoc (Transformation|GetActorLocation))
  (bind (_asynctaskproxy _payload) (Ability|Async|WaitGameplayEventToActor 0 (Variables|EMP|GetEventTag_EMPHit))
    (:then
      (EventDispatchers|BindEventtoOnDamaged (delegate Custom|DamageEvent))
      (EventDispatchers|AssignOnDied (delegate Custom|Destroyed))
      (Ability|ExecuteGameplayCueReplicated (Variables|GCs|GetGC_Spawn))
      (if (Networking|IsServer)
        (CallFunction|SetupRandomTurns)
        (CallFunction|SetupHeightCheck)
        (CallFunction|SetupHeightPicker)
        (if (Variables|LeashCheck|GetHasLeash)
          (CallFunction|SetupLeashCheck))
        (else
          (SoundIndicator|StartSoundIndicatorLooping (Variables|Default|GetFortSoundIndicator) (Transformation|GetActorLocation) (Variables|/Script/Engine.AudioComponent|GetSound (Variables|Audio|GetHoverSound)) 1.0 0)
          (CallFunction|SetupInterpolation))))
    (:EventReceived
      (CallFunction|HandleEMP (Fortnite|GetControllerfromActor (.Instigator (Utilities|Struct|BreakGameplayEventData _payload)))))))

(event EventBeginPlay
  ;[1 1 1] Begin Play
  (AimAssist|RegisterAimAssistTarget self (Variables|Default|GetSphere))
  (if (not (Variables|States|GetAlreadySpawned))
    (CallFunction|OnSpawn)))

(event Custom|DamageEvent
  (params (DamagedActor :type Actor%20Object%20Reference) (Damage :type Float%20%28single-precision%29) (InstigatedBy :type Controller%20Object%20Reference) (DamageCauser :type Actor%20Object%20Reference) (HitLocation :type Vector) (FHitComponent :type Primitive%20Component%20Object%20Reference) (BoneName :type Name) (Momentum :type Vector))
  ;[1 1 1] On Damage
  (Variables|Default|SetPlayer%20Damaging InstigatedBy)
  (Development|PrintString "Damaged" (Variables|Default|GetDebug) (Variables|Default|GetDebug))
  (Utilities|FlowControl|Gate false
    (:Exit
      (Ability|ExecuteGameplayCueWithParamsReplicated (Variables|GCs|GetGC_Damaged) (Ability|GameplayCue|MakeGameplayCueParameters :NormalizedMagnitude 0.0 :RawMagnitude 0.0 :Location HitLocation :Normal "0,0,0" :Instigator InstigatedBy :EffectCauser InstigatedBy :SourceObject InstigatedBy :GameplayEffectLevel 1 :AbilityLevel 1 :bReplicateLocationWhenUsingMinimalRepProxy false))
      (Utilities|FlowControl|Delay (Variables|Default|GetDelayBeforeDamageFX)))))

(event Custom|SetupInterpolation
  (bind _fortprojectilemovement (Variables|Default|GetFortProjectileMovement))
  ;[1 1 1] Set up client interpolation
  (if (Networking|IsServer)
    _
    (elif (Networking|ShouldUseClientInterpolation)
      (Game|Components|ProjectileMovement|Interpolation|SetInterpolatedComponent (Variables|Default|GetFortProjectileMovement) (Variables|Default|GetMesh))
      (Variables|/Script/Engine.ProjectileMovementComponent|SetbInterpMovement true _fortprojectilemovement)
      (Variables|/Script/Engine.ProjectileMovementComponent|SetbInterpRotation true _fortprojectilemovement)
      (Variables|/Script/Engine.ProjectileMovementComponent|SetbSimulationEnabled false _fortprojectilemovement) ; If false, stops projectile simulation, avoiding collision traces
    )))

(event Custom|SpawnMaterialVFX
  (bind _mesh (Variables|Default|GetMesh))
  (bind _returnvalue (Rendering|Material|GetMaterials _mesh))
  (bind _mesh_2 (Variables|Default|GetMesh))
  ;[0.02 0.6 2] Material Changes when Spawn in and Out Occur
  (Variables|SpawnVFX|SetSpawnMatStorage _returnvalue)
  (for (_ _array_index) _returnvalue
    (Rendering|Material|SetMaterial _mesh_2 _array_index "/Game/Athena/Items/Consumables/SuperDingo/Materials/MI_Supply_Drop_Drone_Masked.MI_Supply_Drop_Drone_Masked"))
  (Rendering|Material|SetVectorParameterValueonMaterials _mesh_2 "DissolveGradientBaseLocation" (Transformation|GetWorldLocation (Variables|Default|GetMesh)))
  (Rendering|Material|SetVectorParameterValueonMaterials _mesh_2 "DissolveGradientDirectionalVector" (Variables|SpawnVFX|GetDissolveInVector))
  (Components|Timeline|SetPlayRate (Variables|SpawnVFX|GetSpawnIn) (Utilities|Reciprocal (Variables|SpawnVFX|GetSpawnInDuration)))
  (bind (_direction _dissolve _color) (Timeline|SpawnIn PlayFromStart)
    (:Update
      (Rendering|Material|SetScalarParameterValueonMaterials _mesh_2 "DissolveGradientPercentage" _dissolve)
      (Rendering|Material|SetVectorParameterValueonMaterials _mesh_2 "EmissiveMultiplier" (Math|Conversions|ToVector(LinearColor) _color)))
    (:Finished
      (switch Utilities|FlowControl|Switch|SwitchonETimelineDirection _direction
        (:Forward
          (Rendering|Material|SetScalarParameterValueonMaterials _mesh_2 "DissolveGradientPercentage" 0.0)
          (for (_ _array_index_2) (Rendering|Material|GetMaterials _mesh_2)
            (Rendering|Material|SetMaterial _mesh_2 _array_index_2 (Utilities|Array|Get(acopy) (Variables|SpawnVFX|GetSpawnMatStorage) _array_index_2))))
        (:Backward
          (Development|SetHiddeninGame (Variables|Default|GetMesh) true false))))))

(event Custom|DeSpawnMaterialVFX
  (bind _mesh (Variables|Default|GetMesh))
  (bind _mesh_2 (Variables|Default|GetMesh))
  ;[0.02 0.6 2] Material Changes when Spawn in and Out Occur
  (for (_ _array_index) (Rendering|Material|GetMaterials (Variables|Default|GetMesh))
    (Rendering|Material|SetMaterial _mesh_2 _array_index "/Game/Athena/Items/Consumables/SuperDingo/Materials/MI_Supply_Drop_Drone_Masked.MI_Supply_Drop_Drone_Masked"))
  (Rendering|Material|SetVectorParameterValueonMaterials _mesh_2 "DissolveGradientBaseLocation" (Transformation|GetWorldLocation (Variables|Default|GetMesh)))
  (Rendering|Material|SetVectorParameterValueonMaterials _mesh_2 "DissolveGradientDirectionalVector" (Variables|SpawnVFX|GetDissolveOutVector))
  (Components|Timeline|SetPlayRate (Variables|SpawnVFX|GetSpawnIn) (Utilities|Reciprocal (Variables|SpawnVFX|GetSpawnOutDuration)))
  (Timeline|SpawnIn ReverseFromEnd))

(fn CheckHitByVehicle
  (params (OtherActor :type Actor%20Object%20Reference))
  (bind _asfort_vehicle_interface (Utilities|Casting|CastToFortVehicleInterface OtherActor))
  (Utilities|IsValid (Vehicle|GetDriver _asfort_vehicle_interface)
    (:Is%20Valid
      (CallFunction|Destroyed 0 0.0 (Pawn|GetController (Vehicle|GetDriver _asfort_vehicle_interface))))
    (:Is%20Not%20Valid
      (CallFunction|Destroyed 0))))

(fn HeightCheckCalc
  (bind _fortprojectilemovement (Variables|Default|GetFortProjectileMovement))
  (bind (_velocity_x _velocity_y) (Math|Vector|BreakVector (Variables|/Script/Engine.MovementComponent|GetVelocity_X _fortprojectilemovement)))
  (bind (_outhit _returnvalue) (Collision|LineTraceByProfile (Transformation|GetWorldLocation (Variables|Default|GetSphere)) (Math|Vector|vector+vector (Transformation|GetWorldLocation (Variables|Default|GetSphere)) (Math|Vector|MakeVector 0.0 0.0 (* (+ (Variables|HeightCheck|GetTargetHeight) (LocalVariables|GetHeightDownwardProjectionAdd)) -1.0)) (* (Transformation|GetForwardVector (Variables|Default|GetSphere)) (LocalVariables|GetHeightCheckForwardProjection))) "FortTrajectoryBlocksWater"))
  (if _returnvalue
    (LocalVariables|SetGoDown (>= (.Distance (Collision|BreakHitResult _outhit)) (Variables|HeightCheck|GetTargetHeight)))
    (if (Math|Float|InRange(Float) (.Distance (Collision|BreakHitResult _outhit)) (Variables|HeightCheck|GetTargetHeight) (+ (ScalableFloat|GetValueatLevel (LocalVariables|GetHeightCheckVarience)) (Variables|HeightCheck|GetTargetHeight)))
      (Development|PrintString "No Change" (Variables|Default|GetDebug) (Variables|Default|GetDebug))
      (return "0, 0, 0" false)
      (else
        (LocalVariables|SetAddedZVel (* (Math|Float|Lerp (LocalVariables|GetMinVel) (LocalVariables|GetMaxVel) (/ (Math|Float|Absolute(Float) (- (.Distance (Collision|BreakHitResult _outhit)) (Variables|HeightCheck|GetTargetHeight))) (LocalVariables|GetHeightVarience))) (select (LocalVariables|GetGoDown) -1.0 1.0)))))
    (else
      (LocalVariables|SetGoDown true)
      (LocalVariables|SetAddedZVel (* (LocalVariables|GetMaxVel) -1.0))
      (Development|PrintString "MaxHeight" (Variables|Default|GetDebug) (Variables|Default|GetDebug))))
  (Utilities|IsValid _fortprojectilemovement
    (:Is%20Valid
      (bind _output_get (Variables|SetVelOut_X (Math|Vector|MakeVector _velocity_x _velocity_y (LocalVariables|GetAddedZVel))))
      (return (* (Math|Vector|Normalize _output_get) (Math|Vector|VectorLength (Variables|/Script/Engine.MovementComponent|GetVelocity _fortprojectilemovement))) true))
    (:Is%20Not%20Valid
      (return "0, 0, 0" false))))

(fn MovingCheck
  (bind _returnvalue (Transformation|GetActorLocation))
  ;[1 1 1] Check to see if we have moved recently. Helps to fix issues when the drone gets stuck in objects.
  (if (not (Variables|HitRot|GetHitRotating))
    (bind _distance (Class|MissionBlueprintFunctionLibrary|DistanceBetweenTwoVectors (Variables|MovementFailsafe|GetLastLocationCheck) _returnvalue))
    (if (< _distance (Variables|MovementFailsafe|GetLocationCheckDistance))
      (CallFunction|ResumeMovement)
      (Development|PrintString "Resume Movement")
      (else
        (Variables|MovementFailsafe|SetLastLocationCheck _returnvalue)))))

(fn QuestProgress
  (params (PlayerPawn :type Actor%20Object%20Reference))
  (if (Networking|IsServer)
    (Quest|SendComplexCustomStatEvent (Quests|GetQuestManager (Fortnite|GetFortPlayerControllerfromActor PlayerPawn) "Invalid") 0 "(GameplayTags=((TagName=\"Athena.Quests.SupplyDropDrone\")))" "(GameplayTags=)" 1)))

(fn ConstructionScript
  (bind _row_speed (Variables|Default|GetRow_Speed))
  (bind _returnvalue (ScalableFloat|GetValueatLevel _row_speed))
  (bind _rarity (Variables|Rarity|GetRarity))
  (bind _raritytocolormap (Variables|Rarity|GetRarityToColorMap))
  (bind (_value _returnvalue_2) (Utilities|Map|Find _raritytocolormap _rarity))
  (bind _fortprojectilemovement (Variables|Default|GetFortProjectileMovement))
  (Utilities|IsValid _fortprojectilemovement
    (:Is%20Valid
      (Variables|/Script/Engine.ProjectileMovementComponent|SetMaxSpeed _returnvalue _fortprojectilemovement)
      (bind _output_get (Variables|/Script/Engine.ProjectileMovementComponent|SetInitialSpeed _returnvalue _fortprojectilemovement))
      (Variables|Default|SetBaseSpeed _output_get))
    (:Is%20Not%20Valid
      _))
  (if _returnvalue_2
    (Variables|Rarity|SetBoxColor _value)
    (else
      _))
  (|CallParentFunction))
