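; Server-side spawner actor that places SuperDingo drones in Golden POIs, or anywhere if set to always spawn.
; After playlist data is ready it checks eligibility, then spawns immediately or waits for notification.
; An optional cap combines a POI-size amount with a bonus, then a delayed EQS query supplies locations.
; Drones spawn Legendary on staggered frames with height and leash values from scalable floats, and the spawner destroys itself.
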(event Custom|BeginSpawning
  (Development|PrintString "Golden POI drone spawner begin spawn")
  ;[1 1 1] Check if we limit spawns if so set the initial spawn amount to the max spawn amount then after a delay run the EQS
  (CallFunction|SelectSpawnerPOISize)
  (if (Variables|DingoSpawnerLimit|GetHasSpawnLimit)
    (Variables|DingoSpawnerLimit|SetMaxAmountToSpawn (+ (Variables|DingoSpawnerLimit|GetBonusMaxSpawnAmount) (Variables|DingoSpawnerLimit|GetAmountToSpawnFromSheets)))
    (else
      _))
  (Utilities|FlowControl|Delay (Variables|DingoSpawning|GetDelayBeforeSpawn))
  (CallFunction|RunEqs))

(event Custom|RunEqs
  ;[1 1 1] Run EQS and select if we are doing the initial spawn or respawns.
  (bind _returnvalue (AI|EQS|RunEQSQuery (Variables|DingoSpawning|GetEQS) self "AllMatching"))
  (AI|EQS|BindEventtoOnQueryFinished _returnvalue (delegate Custom|QueryFinished)))

(event Custom|QueryFinished
  (params (QueryInstance :type EQS%20Query%20Instance%20Object%20Reference) (QueryStatus :type EEnvQueryStatus%20Enum :default "Processing"))
  ;[1 1 1] Run EQS and select if we are doing the initial spawn or respawns.
  (bind (_resultlocations _returnvalue) (AI|EQS|GetQueryResultsasLocations QueryInstance))
  (if _returnvalue
    (CallFunction|InitalSpawn _resultlocations)))

(event EventBeginPlay
  ;[1 1 1] Begin Play - Check If enabled and in Golden POI, If not, Kill
  (if (Networking|IsServer)
    (Class|FortAsyncActionCurrentPlaylistReady|PlaylistDataReadyAsync
      (:OnReady
        (Utilities|Enabled
          (:true
            (bind _returnvalue (Fortnite|POI|IsActorWithinGoldenPOI self))
            (if (or _returnvalue (Variables|DingoSpawning|GetAlwaysSpawn))
              (if (not (Variables|Default|GetWaitForNotificationToSpawn))
                (CallFunction|BeginSpawning))
              (else
                _)))
          (:false
            _))
        (Actor|DestroyActor)))))

(event Custom|Spawn
  (params (SpawnTransform :type Transform))
  (bind (_heightmax _heightmin _leashdistance) (CallFunction|GetRowsforSpawn))
  ;[1 1 1] Spawns
  (bind _result (ArrayMacros|SafeArrayGet (Variables|DingoSpawning|GetSuperDingoClasses) (ScalableFloat|GetValueasInteger (Variables|DingoSpawning|GetRow_VariantToSpawnIndex)))
    (:%20%20
      (Game|SpawnActorfromClass _result SpawnTransform "AdjustIfPossibleButAlwaysSpawn" "SelectDefaultAtRuntime" self (CallFunction|RollRarity) _heightmax _heightmin true _leashdistance (Variables|DingoVars|GetDespawnAndRespawnTime) 0 0)
      (Development|PrintString "SpawnDrone"))))

(event Custom|InitalSpawn
  (params (Locations :type Array%20of%20Vectors))
  (bind (_output_x _output_y _output_z) (Math|Vector|BreakVector (Utilities|Array|Get(aref) Locations (Variables|DingoSpawning|GetLoopInt))))
  (bind _loopint (Variables|DingoSpawning|GetLoopInt))
  ;[1 1 1] Spawns
  (if (not (>= (Variables|DingoSpawning|GetLoopInt) (Variables|DingoSpawnerLimit|GetMaxAmountToSpawn)))
    (if (< _loopint (Utilities|Array|Length Locations))
      (CallFunction|Spawn (Math|Transform|MakeTransform (Math|Vector|MakeVector _output_x _output_y (+ _output_z (Variables|DingoVars|GetHeightOffset))) (Math|Rotator|MakeRotator 0.0 0.0 (Math|Random|RandomFloatinRange 0.0 3.6e+02)) "1.000000,1.000000,1.000000"))
      (Utilities|FlowControl|Delay (Math|Random|RandomFloatinRange 0.01 0.1)) ; Do the spawns on different frames
      (Math|Integer|IncrementInt (Variables|DingoSpawning|GetLoopInt)
        (:%20%20
          _))))
  (Utilities|FlowControl|Delay 0.2)
  (Actor|DestroyActor))

(fn GetRowsForSpawn
  (return (ScalableFloat|GetValueatLevel (Variables|DingoVars|GetRow_FlyHeightMax)) (ScalableFloat|GetValueatLevel (Variables|DingoVars|GetRow_FlyHeightMin)) (ScalableFloat|GetValueatLevel (Variables|DingoVars|GetRow_LeashDistance))))

(fn RollRarity
  (return "Legendary"))

(fn SelectSpawnerPOISize
  (switch Utilities|FlowControl|Switch|SwitchonEnum_SuperDingo_SpawnerPOISize (Variables|DingoSpawnerLimit|GetPOISize)
    (:NewEnumerator0
      (Variables|DingoSpawnerLimit|SetAmountToSpawnFromSheets (Math|Float|Truncate (ScalableFloat|GetValueatLevel (Variables|DingoSpawnerLimit|GetPOISmallSpawnAmount)))))
    (:NewEnumerator2
      (Variables|DingoSpawnerLimit|SetAmountToSpawnFromSheets (Math|Float|Truncate (ScalableFloat|GetValueatLevel (Variables|DingoSpawnerLimit|GetPOIMedSpawnAmount)))))
    (:NewEnumerator1
      (Variables|DingoSpawnerLimit|SetAmountToSpawnFromSheets (Math|Float|Truncate (ScalableFloat|GetValueatLevel (Variables|DingoSpawnerLimit|GetPOILargeSpawnAmount))))))
  (return))

(fn ConstructionScript
  (Rendering|Debug|DrawDebugCircle (Transformation|GetActorLocation) (+ (ScalableFloat|GetValueatLevel (Variables|DingoVars|GetRow_LeashDistance)) (ScalableFloat|GetValueatLevel (Variables|DingoVars|GetRow_MaxSpawnDistance))) 32 "(R=1.000000,G=0.000000,B=0.045438,A=1.000000)" 5.0 0.0 "0.000000,1.000000,0.000000" "1.000000,0.000000,0.000000")
  ;[1 1 1] Hide Debug info if it is unneeded.
  (if (not (Variables|Default|GetDebug))
    (Rendering|SetVisibility (Variables|Default|GetArrow) false false)))
