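; EQS query context for the SuperDingo consumable that supplies a single randomized location relative to the querier actor.
; It rotates the actor's forward vector by a random yaw, scales it by a random distance, adds it to the actor's position, then raises Z by a local offset.
; Per the inline note, this starts a trace from the sky and varies results for players picking up maps at the same point.
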(fn ProvideSingleLocation
  (params (QuerierObject :type Object%20Reference) (QuerierActor :type Actor%20Object%20Reference))
  (bind (_returnvalue_x _returnvalue_y _returnvalue_z) (Math|Vector|BreakVector (+ (Transformation|GetActorLocation QuerierActor) (* (Math|Vector|RotateVectorAroundAxis (Transformation|GetActorForwardVector QuerierActor) (Math|Random|RandomFloatinRange 0.0 3.6e+02) "0, 0,1.000000") (Math|Random|RandomFloatinRange 512.0 1024.0)))))
  ;[1 1 1] Start trace from the sky and offset it randomly so that even if two players pick up maps at the same exact point the pounts will be slightly different.
  (Utilities|IsValid QuerierActor
    (:Is%20Valid
      (return _returnvalue_x _returnvalue_y (+ _returnvalue_z (LocalVariables|GetAddedZLoc))))))
