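# Material function that computes procedural World Position Offset motion for the SuperDingo supply drop drone.
# Time-driven sines and cosines feed four outputs: vertical bob, normalized-vector vibration, base rotation and lagged rotation.
# Both rotations oscillate around a wandering axis through a parameterized pivot relative to actor position, using RotateAboutAxis.
# RotationLag repeats the base rotation with a Lag time offset, and the bob shares its speed parameter with the axis wander.
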MaterialFunction /Game/Athena/Items/Consumables/SuperDingo/Materials/MF_SupplyDropDroneWPO
graph {nodes [
  Constant() {r=0.09}
  Constant() {r=0.01}
  Constant() {r=0.1714}
  Constant() {r=0.4}
  Constant() {r=8}
  Constant() {r=-0.3}
  n1 = Time()
  n2 = ScalarParameter() {defaultValue=0.4, parameterName=SpeedBallon_2}
  n3 = Multiply(A=@n1, B=@n2)
  n4 = Sine(None=@n3)
  n5 = ConstantBiasScale(None=@n4)
  n6 = Constant3Vector() {constant=<rgba 0 0 1 0>}
  FunctionOutput(None=Multiply(A=Multiply(A=@n5, B=@n6), B=ScalarParameter() {defaultValue=7, parameterName=VerticalMovement})) {outputName=BaseVerticalMovement, sortPriority=1}
  n7 = Multiply(A=Time(), B=ScalarParameter() {defaultValue=8, parameterName=SpeedVibration})
  n8 = AppendVector(A=Sine(None=@n7), B=AppendVector(A=Sine(None=@n7), B=Cosine(None=@n7)))
  FunctionOutput(None=Multiply(A=Normalize(VectorInput=@n8), B=ScalarParameter() {defaultValue=0.4, parameterName=VibrationIntensity})) {outputName=Vibration}
  n9 = AppendVector(A=@n4, B=AppendVector(A=Sine(None=@n3), B=Cosine(None=@n3)))
  n10 = ScalarParameter() {defaultValue=0.2, parameterName=OverallRotationSpeed}
  n11 = Sine(None=Multiply(A=@n1, B=@n10))
  n12 = ScalarParameter() {defaultValue=0.01, parameterName=OverallRotation}
  n13 = Add(A=VectorParameter() {parameterName=RotationCenter, defaultValue=<rgba 0 0 60 1>}.RGB, B=ActorPositionWS())
  n14 = WorldPosition() {worldPositionShaderOffset=WPT_ExcludeAllShaderOffsets}
  FunctionOutput(None=RotateAboutAxis(NormalizedRotationAxis=Normalize(VectorInput=@n9), RotationAngle=Multiply(A=@n11, B=@n12), PivotPoint=@n13, Position=@n14.XYZ)) {outputName=BaseRotation, sortPriority=2}
  n15 = Add(A=Time(), B=ScalarParameter() {defaultValue=0.2, parameterName=Lag})
  n16 = Multiply(A=@n15, B=@n2)
  n17 = AppendVector(A=Sine(None=@n16), B=AppendVector(A=Sine(None=@n16), B=Cosine(None=@n16)))
  n18 = Sine(None=Multiply(A=@n15, B=@n10))
  FunctionOutput(None=RotateAboutAxis(NormalizedRotationAxis=Normalize(VectorInput=@n17), RotationAngle=Multiply(A=@n18, B=@n12), PivotPoint=@n13, Position=@n14.XYZ)) {outputName=RotationLag, sortPriority=3}
 ]}
