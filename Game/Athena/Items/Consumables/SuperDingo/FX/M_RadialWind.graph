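# Unlit translucent two-sided material for the SuperDingo consumable's radial wind effect, usable on Niagara mesh particles.
# Panned smoke and turbulence textures share noise-distorted UVs, masked by curve-atlas ramps on UV1, driving depth-faded opacity.
# "Use With Particles" (default on) takes color and alpha from particles and zeroes vertex offset.
# Otherwise it uses time-of-day tinting, Fresnel edge fade (constant on mobile), normal-based offset and supply-drop-drone vibration.
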Material /Game/Athena/Items/Consumables/SuperDingo/FX/M_RadialWind
graph {
 properties {
  bUsedWithNiagaraMeshParticles=true
  blendMode=BLEND_Translucent
  pixelDepthOffsetMode=PDOM_Legacy
  shadingModel=MSM_Unlit
  twoSided=true
  usedShadingModels=Unlit
 }
 nodes [
  n1 = ParticleColor()
  n2 = MaterialFunctionCall(
   Color_Morning=VectorParameter() {parameterName=MorningTint, defaultValue=<rgba 1.15 0.8884 0.805 0>}.RGB
   Color_Day=VectorParameter() {parameterName=DayTint, defaultValue=<rgba 1.0812 1.2 1.1958 0>}.RGB
   Color_Evening=VectorParameter() {parameterName=EveningTint, defaultValue=<rgba 1.1 0.7845 0.77 1>}.RGB
   Color_Night=VectorParameter() {parameterName=NightTint, defaultValue=<rgba 0.2656 0.85 0.85 0>}.RGB
  ) {materialFunction=&Color_Time_of_Day}
  n3 = ScalarParameter() {defaultValue=0.65, parameterName=TimeOfDayColorContribution}
  n4 = Multiply(
   A=LinearInterpolate(A=VectorParameter() {parameterName=BaseColor, defaultValue=<rgba 0.4177 1 0.9 0>}.RGB, B=@n2.Result, Alpha=@n3)
   B=ScalarParameter() {defaultValue=0.02, parameterName=ShadowMult}
  ) {constB=0.1}
  n5 = MaterialFunctionCall(Float2=TextureCoordinate() {uTiling=6, vTiling=0.5}) {materialFunction=&BreakOutFloat2Components}
  n6 = TextureSample() {samplerType=SAMPLERTYPE_LinearColor, texture=&T_NoisePack_LinearColor}
  n7 = AppendVector(A=@n6.R, B=@n6.R)
  n8 = Add(A=AppendVector(A=@n5.R, B=Multiply(A=@n5.G, B=@n5.G)), B=Multiply(A=Subtract(A=@n7) {constB=0.35}) {constB=0.02})
  n9 = VectorParameter() {parameterName=VU, defaultValue=<rgba 0.1 1 0 0.8>}
  n10 = Panner(Coordinate=Multiply(A=@n8, B=ComponentMask(None=@n9.RGB) {g=true, r=true}), Speed=AppendVector(A=@n9.B, B=@n9.B)) {speedX=0.1, speedY=0.5}
  n11 = VectorParameter() {parameterName=VU2, defaultValue=<rgba 0.05 1 0 1>}
  n12 = ComponentMask(None=@n11.RGB) {g=true, r=true}
  n13 = TextureSample(UVs=Panner(Coordinate=Multiply(A=@n8, B=@n12), Speed=AppendVector(A=@n11.B, B=@n11.B)) {speedX=0.1, speedY=0.5}) {texture=/Game/Global/Textures/Default/Caustics/T_Turbulence_Seamless}
  n14 = MaterialFunctionCall(Float2=TextureCoordinate() {coordinateIndex=1}) {materialFunction=&BreakOutFloat2Components}
  n15 = CurveAtlasRowParameter(CurveTime=@n14.R) {
   atlas=&AthenaCurveAtlas01
   curve=/Game/Athena/Environments/Blueprints/DudeBro/HVAC_FX/Curve_Ramps
   defaultValue=84
  }
  n16 = CurveAtlasRowParameter(CurveTime=@n14.G) {
   atlas=&AthenaCurveAtlas01
   curve=&Curve_Ramps
   defaultValue=84
  }
  n17 = Multiply(
   A=Multiply(
    A=TextureSample(UVs=@n10) {texture=&T_Inky_Smoke_Tile}.R
    B=Power(Base=@n13.R) {constExponent=0.7}
   )
   B=Multiply(A=@n15.A, B=@n16.B)
  )
  n18 = Multiply(
   A=LinearInterpolate(B=@n2.Result, Alpha=@n3) {constA=1.2}
   B=QualitySwitch(Default=LinearInterpolate(Alpha=@n17), Low=Add(B=@n17) {constA=0.15})
  )
  MaterialFunctionCall(Emissive=StaticSwitchParameter(True=@n1.RGB, False=Add(A=@n4, B=@n18) {constB=0.2}) {defaultValue=true, parameterName=Use With Particles}) {materialFunction=&MF_FX_EmissiveAdjustments}
  n19 = ScalarParameter() {defaultValue=1, parameterName=Opacity}
  n20 = ShadingPathSwitch(
   Default=MaterialFunctionCall(
    Axis=MaterialFunctionCall() {materialFunction=&TangentBasis}.Y V
    Exponent=ScalarParameter() {defaultValue=2, parameterName=AxisExp}
   ) {desc=Edge fade, materialFunction=&AxisAlignedFresnel}.Fresnel
   Mobile=Constant() {r=0.3}
  ) {desc=Migrated from FeatureLevelSwitch}
  Multiply(
   A=StaticSwitchParameter(True=@n1.A, False=Multiply(A=@n19, B=@n20)) {defaultValue=true, parameterName=Use With Particles}
   B=DepthFade(Opacity=@n17) {fadeDistanceDefault=20}
  )
  Power(Base=@n14.G)
  n21 = TextureCoordinate()
  n22 = Multiply(A=OneMinus(None=ComponentMask(None=@n21) {g=true}), B=ScalarParameter() {defaultValue=-10, parameterName=WPO_Amount})
  n23 = VertexNormalWS()
  n24 = MaterialFunctionCall() {materialFunction=/Game/Athena/Items/Consumables/SuperDingo/Materials/MF_SupplyDropDroneWPO}
  StaticSwitchParameter(True=Constant(), False=Add(A=Multiply(A=@n22, B=@n23), B=Add(A=@n24.Vibration, B=@n24.Vibration))) {defaultValue=true, parameterName=Use With Particles}
 ]
}
