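# Unlit, additive, two-sided wind-streak material for SuperDingo propeller effects, usable on Niagara mesh particles.
# Two samples of slanted, panning turbulence drive time-of-day tinted color, with optional HSV adjustment and position-based per-component variation.
# Opacity combines axis-aligned Fresnel, UV end masks, depth fade and optional distance fades, with cheaper quality and mobile branches.
# World position offset pushes vertices along normals by V, adding supply-drone vibration or switchable Welderbot bob.
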Material /Game/Athena/Items/Consumables/SuperDingo/FX/M_Wind_Propeller_SPline.M_Wind_Propeller_Spline
graph {
 properties {
  bUsedWithNanite=true
  bUsedWithNiagaraMeshParticles=true
  blendMode=BLEND_Additive
  pixelDepthOffsetMode=PDOM_Legacy
  shadingModel=MSM_Unlit
  twoSided=true
  usedShadingModels=Unlit
 }
 nodes [
  Constant()
  Constant()
  MaterialExpressionScalarParameter_8 = ScalarParameter() {defaultValue=1, parameterName=Object fade account for Z}
  Convert(None=@MaterialExpressionScalarParameter_8) {
   convertInputs [{
     type=Scalar
     expressionInput {
      expression=&self:MaterialExpressionScalarParameter_8
      inputName=None
      mask=0
      maskA=0
      maskB=0
      maskG=0
      maskR=0
      outputIndex=0
     }
    }]
   convertOutputs [{type=Vector3, defaultValue=<rgba 1 1 0 1>}]
  }
  n1 = VectorParameter() {parameterName=BaseColor, defaultValue=<rgba 0.4177 1 0.9 0>}
  n2 = MaterialFunctionCall(
   Color_Morning=VectorParameter() {parameterName=MorningTint, defaultValue=<rgba 1.15 0.8884 0.805 0>}.RGB
   Color_Day=VectorParameter() {parameterName=DayTint, defaultValue=<rgba 1.0812 1.2 1.1958 0>}.RGB
   Color_Evening=VectorParameter() {parameterName=EveningTint, defaultValue=<rgba 1.1 0.7845 0.77 1>}.RGB
   Color_Night=VectorParameter() {parameterName=NightTint, defaultValue=<rgba 0.2656 0.85 0.85 0>}.RGB
  ) {materialFunction=&Color_Time_of_Day}
  n3 = ScalarParameter() {defaultValue=0.65, parameterName=TimeOfDayColorContribution}
  n4 = ScalarParameter() {defaultValue=3, parameterName=XTile}
  n5 = ScalarParameter() {defaultValue=0.5, parameterName=YTile}
  n6 = ComponentMask(None=TextureCoordinate()) {g=true}
  n7 = ScalarParameter() {defaultValue=1, parameterName=UVSlant}
  n8 = Constant()
  n9 = Add(A=Multiply(A=TextureCoordinate(), B=AppendVector(A=@n4, B=@n5)), B=Multiply(A=Saturate(None=@n6), B=AppendVector(A=@n7, B=@n8)))
  n10 = Add(A=@n9) {constB=0.5}
  n11 = ScalarParameter() {defaultValue=0.45, parameterName=PanSpeedX}
  n12 = Frac(None=Multiply(A=ObjectPositionWS()))
  n13 = Add(A=@n11, B=NamedRerouteDeclaration(None=ComponentMask(None=@n12) {g=true}) {name=PerComponentRandom, nodeColor=<rgba 0.1403 1 0.8629 1>})
  n14 = AppendVector(
   A=LinearInterpolate(A=@n11, B=Clamp(None=@n13) {maxDefault=0.6, minDefault=0.45}, Alpha=ScalarParameter() {parameterName=UsePerComponentRandom?})
   B=ScalarParameter() {defaultValue=-0.9, parameterName=PanSpeedY}
  )
  n15 = LinearInterpolate(
   A=TextureSampleParameter2D(UVs=Panner(Coordinate=@n10, Speed=@n14)) {parameterName=Texture, texture=/Game/Effects/Fort_Effects/Textures/Plasma/T_Turbulence_Stretched}.G
   B=TextureSampleParameter2D(UVs=Panner(Coordinate=@n9, Speed=@n14)) {parameterName=Texture, texture=&T_Turbulence_Stretched}.G
  )
  n16 = LinearInterpolate(Alpha=@n15)
  n17 = Add(B=@n15) {constA=0.15}
  n18 = Add(
   A=Multiply(A=LinearInterpolate(A=@n1.RGB, B=@n2.Result, Alpha=@n3), B=ScalarParameter() {defaultValue=0.02, parameterName=ShadowMult}) {constB=0.1}
   B=Multiply(A=LinearInterpolate(B=@n2.Result, Alpha=@n3) {constA=1.2}, B=QualitySwitch(Default=@n16, Low=@n17, High=@n17, Medium=@n17))
  ) {constB=0.2}
  n19 = VectorParameter() {parameterName=HSV Adjustment, channelNames {a=NA, b=Value, g=Sat, r=Hue}, defaultValue=<rgba 0 1 1 0>}
  n20 = Multiply(
   A=Desaturation(
    None=MaterialFunctionCall(Hue Shift Percentage=@n19.Hue, Texture=@n18) {materialFunction=&HueShift}.Result
    Fraction=OneMinus(None=@n19.Sat)
   )
   B=@n19.Value
  )
  n21 = VectorParameter() {parameterName=HSV MaskOptions, channelNames {a=constrast, b=add, g=mul, r=pow}, defaultValue=<rgba 1 1 0 0>}
  n22 = Multiply(
   A=Power(Base=MaterialFunctionCall() {materialFunction=&LinearGradient}.VGradient, Exp=@n21.pow)
   B=@n21.mul
  )
  n23 = Saturate(None=MaterialFunctionCall(In=Add(A=@n22, B=@n21.add), Contrast=@n21.constrast) {materialFunction=&CheapContrast}.Result)
  MaterialFunctionCall(Emissive=LinearInterpolate(A=@n18, B=LinearInterpolate(A=@n18, B=@n20, Alpha=@n23), Alpha=ScalarParameter() {parameterName=Use HSV Adjustments})) {materialFunction=&MF_FX_EmissiveAdjustments}
  n24 = Multiply(A=ComponentMask(None=TextureCoordinate()) {g=true}, B=ScalarParameter() {defaultValue=-2, parameterName=WPO_Amount})
  n25 = VertexNormalWS()
  n26 = MaterialFunctionCall() {materialFunction=/Game/Athena/Environments/Volcano/Meshes/Welderbot/MF_Welderbot_WPOBob}
  n27 = MaterialFunctionCall() {materialFunction=/Game/Athena/Items/Consumables/SuperDingo/Materials/MF_SupplyDropDroneWPO}
  n28 = Add(A=@n27.Vibration, B=@n27.Vibration)
  Reroute(None=Add(A=Multiply(A=@n24, B=@n25), B=StaticSwitchParameter(True=@n26.Result, False=@n28) {parameterName=WelderBotWPO}))
  n29 = Multiply(A=@n15, B=ScalarParameter() {defaultValue=0.2, parameterName=MiddleFadeOpacity}) {constB=0.5}
  n30 = Multiply(A=ScalarParameter() {defaultValue=0.75, parameterName=add}) {constB=0.15}
  n31 = Add(A=@n30) {constB=0.02}
  n32 = LinearInterpolate(
   A=ScalarParameter() {defaultValue=2, parameterName=MaxFresnelExp}
   B=ScalarParameter() {defaultValue=1, parameterName=MinFresnelExp}
   Alpha=@n15
  )
  n33 = ScalarParameter() {parameterName=Use Fres On Med Low}
  n34 = LinearInterpolate(A=Constant() {r=2}, B=@n32, Alpha=@n33)
  n35 = MaterialFunctionCall(
   Axis=MaterialFunctionCall() {materialFunction=&TangentBasis}.Y V
   Exponent=QualitySwitch(Default=@n32, Low=@n34, Medium=@n34)
  ) {desc=Edge fade, materialFunction=&AxisAlignedFresnel}
  n36 = Constant() {r=0.3}
  n37 = Multiply(
   A=QualitySwitch(Default=Add(A=@n29, B=@n30), Low=@n31, High=@n31, Medium=@n31)
   B=ShadingPathSwitch(Default=@n35.Fresnel, Mobile=LinearInterpolate(A=@n36, B=@n35.Fresnel, Alpha=@n33)) {desc=Migrated from FeatureLevelSwitch}
  )
  n38 = DepthFade(Opacity=@n37, FadeDistance=ScalarParameter() {defaultValue=150, parameterName=DepthFadeDist})
  n39 = ComponentMask(None=TextureCoordinate()) {g=true}
  n40 = OneMinus(None=@n39)
  n41 = ScalarParameter() {parameterName=ReverseMask}
  n42 = Multiply(
   A=Power(Base=LinearInterpolate(A=@n40, B=@n39, Alpha=@n41), Exp=ScalarParameter() {defaultValue=3, parameterName=End Fade Contrast})
   B=ScalarParameter() {defaultValue=1, parameterName=End Fade Scale}
  )
  n43 = OneMinus(None=ComponentMask(None=TextureCoordinate()) {g=true})
  n44 = Multiply(A=LinearInterpolate(A=@n42, B=Multiply(A=@n42, B=@n43), Alpha=ScalarParameter() {parameterName=UseBottomMask?})) {constB=2}
  n45 = Multiply(
   A=Multiply(A=ShadingPathSwitch(Default=@n38, Mobile=@n37) {desc=Migrated from FeatureLevelSwitch}, B=Saturate(None=@n44))
   B=ScalarParameter() {defaultValue=0.3, parameterName=OpacityMult}
  ) {constB=2}
  n46 = NamedRerouteUsage()
  n47 = LinearInterpolate(A=@n45, B=Multiply(A=@n45, B=Add(A=@n45, B=@n46)), Alpha=ScalarParameter() {parameterName=UsePerComponentRandom?})
  n48 = Multiply(
   A=@n47
   B=MaterialFunctionCall(
    Fade Length=ScalarParameter() {defaultValue=1000, parameterName="Fade_Start_Distance (S)"}
    Fade Offset=ScalarParameter() {defaultValue=500, parameterName="Fade_End_Distance (S)"}
   ) {materialFunction=&CameraDepthFade}.Result
  )
  n49 = ScalarParameter() {parameterName=Camera Fade Out Length}
  n50 = ScalarParameter() {parameterName=Camera Fade Out Offset}
  n51 = LinearInterpolate(
   B=OneMinus(None=MaterialFunctionCall(Fade Length=@n49, Fade Offset=@n50) {materialFunction=&CameraDepthFade}.Result)
   Alpha=ScalarParameter() {parameterName=Use Dist Fade Out}
  ) {constA=1}
  n52 = CameraPositionWS()
  n53 = ObjectPositionWS()
  n54 = InverseLinearInterpolate(
   A=ScalarParameter() {parameterName=Object distance fade Close}
   B=ScalarParameter() {parameterName=Object distance fade Far}
   Value=Distance(A=ComponentMask(None=@n52) {g=true, r=true}, B=ComponentMask(None=@n53) {g=true, r=true})
  ) {bClampResult=true}
  n55 = ScalarParameter() {parameterName=Use whole Object Fade}
  StaticSwitchParameter(True=Multiply(A=Multiply(A=@n48, B=@n51), B=LinearInterpolate(B=@n54, Alpha=@n55) {constA=1}), False=@n47) {parameterName=UseDistanceFade?}
  VectorParameter() {parameterName=Manual Pos WS}
 ]
}
