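# Niagara ribbon trail for the falling SuperDingo consumable box, drawn as a pale blue, translucent air-cut streak.
# A single AirCut_Trail emitter spawns ribbon points by distance traveled, with movement tolerance and a max movement threshold.
# Each short-lived point accelerates upward through a constant Z acceleration set at spawn, while a linear curve fades its alpha to zero.
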NiagaraSystem /Game/Athena/Items/Consumables/SuperDingo/FX/NS_SuperDingo_BoxFalling
dependencies {usedDataInterfaces [&NiagaraDataInterfaceCurve], usedDynamicInputs [], usedModules [&SystemState]}
emitters {AirCut_Trail {
  emitterUpdateScript [
   &EmitterState()
   &SpawnPerUnit(
    Max Movement Threshold=100
    Module.Use Max Movement Threshold=true
    Module.Use Movement Tolerance=true
    Movement Tolerance=30
    Spawn Spacing:NiagaraFloat=&Divide_Float(A=1, B=&Divide_Float(A=1, B=30))
    Use Spawn Probability=false
   )
  ]
  particleSpawnScript [
   /SetVariables(Particles.Acceleration:Vector3f=<0 0 500>) {isSetParameters=true}
   &InitializeRibbon(Lifetime=1.5, Ribbon Width=&Multiply_Float(A=15, B=2), UsePositionOffset=false, Write Material Random=true)
   &Color(Color=&MakeLinearColorFromVectorAndFloat("Vector (RGB)"=<0.8333 0.9694 1>, "Float (Alpha)"=&ClampFloat(Float=0.25)))
  ]
  particleUpdateScript [
   &ParticleState()
   &ScaleColor(
    Scale Alpha=&FloatFromCurve(FloatCurve=DataInterface() {propertyValues {Curve {keys [<at 0 1>,<at 1 0>]}}})
    ScaleA=true
    ScaleRGB=true
   )
   &AccelerationForce(Acceleration=$Particles.Acceleration)
   &SolveForcesAndVelocity(Clamp Velocity=false, Limit Acceleration=false)
  ]
  rendererClasses [&NiagaraRibbonRendererProperties]
  renderers [{rendererClass=&NiagaraRibbonRendererProperties}]
 }}
summary {
 systemName=NS_SuperDingo_BoxFalling
 emitters [{emitterName=AirCut_Trail, rendererClasses [&NiagaraRibbonRendererProperties]}]
 userVariables []
}
