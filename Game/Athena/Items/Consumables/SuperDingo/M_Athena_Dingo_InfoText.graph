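# Masked, two-sided, lit material for info text on the SuperDingo consumable in Athena.
# It samples the Roboto distance field font, remapping the red channel through a wide signed lerp, a contrast boost suited to sharp glyph edges.
# An AlignMeshToTheCamera function with a custom axis basis provides billboarding math, and a vertex color node is present.
# The graph shows no connections to the material outputs.
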Material /Game/Athena/Items/Consumables/SuperDingo/M_Athena_Dingo_InfoText
graph {
 properties {blendMode=BLEND_Masked, pixelDepthOffsetMode=PDOM_Legacy, twoSided=true, usedShadingModels=Default Lit}
 nodes [
  LinearInterpolate(Alpha=FontSampleParameter() {font=/Engine/EngineFonts/RobotoDistanceField, parameterName=Font}.R) {constA=-90, constB=90}
  MaterialFunctionCall!(
   WorldPosition
   PivotLocation
   WS Normals
   Custom Object Basis 1=Constant3Vector() {constant=<rgba 0 0 1 0>}
   Custom Object Basis 2=Constant3Vector() {constant=<rgba 1 0 0 0>}
   Custom Object Basis 3=Constant3Vector() {constant=<rgba 0 -1 0 0>}
  ) {materialFunction=/Engine/Functions/Engine_MaterialFunctions02/WorldPositionOffset/AlignMeshToTheCamera}
  VertexColor()
 ]
}
