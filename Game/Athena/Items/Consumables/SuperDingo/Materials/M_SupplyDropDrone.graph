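# Default Lit material for the SuperDingo supply drop drone, with flight motion, spinning blades and blinking emissive lights.
# Vertex colors weight vibration, bobbing, rotation and lag from a drone WPO function, scaled by IsFlying.
# Blue vertex color times IsFlying swaps texture-driven surface values for fake blade spin, using scrolling, rotated gradients and a blended normal.
# Time-based sine waves gate the emissive texture into flashing bursts above a dim level.
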Material /Game/Athena/Items/Consumables/SuperDingo/Materials/M_SupplyDropDrone
graph {
 properties {pixelDepthOffsetMode=PDOM_Legacy, usedShadingModels=Default Lit}
 nodes [
  Constant3Vector() {constant=<rgba 5 4 4 4>}
  MaterialFunctionCall!(Animation Gradient, Rez In Percentage) {materialFunction=/Game/Packages/Fortress_SharedMaterials/Base_Material_Functions/CharacterResInEffect}
  n1 = ScalarParameter() {defaultValue=1, parameterName=IsFlying}
  n2 = VertexColor()
  n3 = MaterialFunctionCall() {materialFunction=&/MF_SupplyDropDroneWPO}
  n4 = Multiply(A=@n3.Vibration, B=Add(A=Add(A=@n2.R, B=@n2.R), B=@n2.B))
  n5 = VertexColor()
  n6 = Add(A=@n5.G, B=@n5.G)
  n7 = OneMinus(None=@n6)
  n8 = Add(A=@n3.BaseVerticalMovement, B=Add(A=Multiply(A=@n3.BaseRotation, B=@n7), B=Multiply(A=@n6, B=@n3.RotationLag)))
  Reroute(None=Multiply(A=@n1, B=Add(A=@n4, B=@n8)))
  n9 = Constant3Vector() {constant=<rgba 0 0 1 0>}
  n10 = Transform(None=Constant3Vector() {constant=<rgba 0 0 1 0>}) {transformSourceType=TRANSFORMSOURCE_World, transformType=TRANSFORM_Tangent}
  n11 = Constant() {desc=Cannot be exactly .5, r=0.55}
  n12 = Multiply(A=VertexColor().B, B=@n1)
  Reroute(None=LinearInterpolate(
    A=TextureSample() {samplerType=SAMPLERTYPE_Normal, texture=/Game/Athena/Items/Consumables/SuperDingo/Textures/T_Supply_Drop_Drone_N}.RGB
    B=LinearInterpolate(A=@n9, B=@n10, Alpha=@n11)
    Alpha=@n12
   ))
  n13 = VectorParameter() {parameterName=EmissiveBoost, defaultValue=<rgba 40 32 32 4>}
  n14 = TextureSample() {texture=/Game/Athena/Items/Consumables/SuperDingo/Textures/T_Supply_Drop_Drone_E}
  n15 = Multiply(A=Time(), B=Constant() {r=0.5})
  n16 = Saturate(None=Round(None=Sine(None=@n15)))
  n17 = Sine(None=Multiply(A=Time(), B=Constant() {r=3}))
  n18 = Add(A=Saturate(None=Multiply(A=@n16, B=@n17)), B=ScalarParameter() {defaultValue=0.2, parameterName=Lights-DimAmount})
  Reroute(None=Multiply(A=Multiply(A=@n13.RGB, B=@n14.RGB), B=Saturate(None=@n18)))
  n19 = TextureSample() {texture=/Game/Athena/Items/Consumables/SuperDingo/Textures/T_Supply_Drop_Drone_S}
  n20 = Constant() {r=1}
  n21 = Constant() {r=0.3}
  n22 = ScalarParameter() {defaultValue=4.3, parameterName=BladeCount}
  n23 = TextureCoordinate()
  n24 = Frac(None=Time())
  n25 = Constant() {r=0.25}
  n26 = MaterialFunctionCall(UVs=Add(A=Multiply(A=@n22, B=@n23), B=@n24), "Rotation Angle (0-1)"=@n25) {materialFunction=&CustomRotator}
  n27 = ScalarParameter() {defaultValue=1.4, parameterName=BladeFlashRepeat}
  n28 = MaterialFunctionCall(UVs=Add(A=Multiply(A=@n23, B=@n27), B=Multiply(A=@n24) {constB=-12}), "Rotation Angle (0-1)"=@n25) {materialFunction=&CustomRotator}
  n29 = Saturate(None=Add(
    A=TextureSample(UVs=@n26.Rotated Values) {samplerType=SAMPLERTYPE_LinearColor, texture=/Game/Global/Textures/Default/Masks_and_Gradients/T_VertGradient}.R
    B=TextureSample(UVs=@n28.Rotated Values) {samplerType=SAMPLERTYPE_LinearColor, texture=&T_VertGradient}.R
   ))
  Reroute(None=LinearInterpolate(A=@n19.B, B=LinearInterpolate(A=@n20, B=@n21, Alpha=@n29), Alpha=@n12))
  Reroute(None=@n19.R)
  n30 = Add(A=@n12, B=@n19.G)
  Reroute(None=LinearInterpolate(A=@n19.G, B=Saturate(None=@n30), Alpha=@n12))
  n31 = Constant()
  n32 = Constant() {r=0.4}
  Reroute(None=LinearInterpolate(
    A=TextureSample() {texture=/Game/Athena/Items/Consumables/SuperDingo/Textures/T_Supply_Drop_Drone_D}.RGB
    B=LinearInterpolate(A=@n31, B=@n32, Alpha=@n29)
    Alpha=@n12
   ))
  n33 = TextureCoordinate() {coordinateIndex=1}
  n34 = Time()
  n35 = Add(A=Multiply(A=Constant() {r=3}, B=@n33), B=@n34)
  n36 = Add(A=@n33, B=Multiply(A=@n34, B=Constant() {r=-10}))
  Saturate(None=Add(
    A=TextureSample(UVs=@n35) {samplerType=SAMPLERTYPE_LinearColor, texture=&T_VertGradient}.R
    B=TextureSample(UVs=@n36) {samplerType=SAMPLERTYPE_LinearColor, texture=&T_VertGradient}.R
   ))
 ]
}
